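# Physics asset for the two dangling strings on the NeonCat female medium player body mesh.
# Mirrored left and right chains of four small capsule bodies each hang from one large spine_05 capsule, with no bodies for the 02 string bones.
# Constraints link each segment to the one above, keeping twist tight while the second swing limit widens toward the tips for freer end movement.
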PhysicsAsset /Game/Characters/Player/Female/Medium/Bodies/F_MED_Neoncat/Meshes/F_MED_NeonCat_Body_T_01_String_Physics
bodies {
 dyn_string_l_01 [{shapeType=Capsule, length=2.5459, radius=0.5101, center=<0.2652 0.0007 -0.0608>, rotation=<rot 90 0 0>}]
 dyn_string_l_03 [{shapeType=Capsule, length=2.5459, radius=0.5101, center=<1.1879 0 0>, rotation=<rot 90 0 0>}]
 dyn_string_l_04 [{shapeType=Capsule, length=2.5459, radius=0.5101, center=<1.3593 0 0>, rotation=<rot 90 0 0>}]
 dyn_string_l_05 [{shapeType=Capsule, length=2.0633, radius=0.6785, center=<2.0429 -0.0665 -0.0074>, rotation=<rot 90 0 0>}]
 dyn_string_r_01 [{shapeType=Capsule, length=2.5459, radius=0.5101, center=<0.2652 -0.0007 -0.0604>, rotation=<rot 90 0 0>}]
 dyn_string_r_03 [{shapeType=Capsule, length=2.5459, radius=0.5101, center=<1.1879 0 0.0001>, rotation=<rot 90 0 0>}]
 dyn_string_r_04 [{shapeType=Capsule, length=2.5459, radius=0.5101, center=<1.3593 0 0.0001>, rotation=<rot 90 0 0>}]
 dyn_string_r_05 [{shapeType=Capsule, length=2.0633, radius=0.6785, center=<2.0429 0.0665 -0.0074>, rotation=<rot 90 0 0>}]
 spine_05 [{shapeType=Capsule, length=8.3381, radius=11.4251, center=<-6.7867 1.9158 0>}]
}
constraints [
 {bone1Name=dyn_string_l_01, bone2Name=spine_05, swing1LimitDegrees=45, swing2LimitDegrees=45, twistLimitDegrees=5}
 {bone1Name=dyn_string_l_03, bone2Name=dyn_string_l_01, swing1LimitDegrees=45, swing2LimitDegrees=45, twistLimitDegrees=5}
 {bone1Name=dyn_string_l_04, bone2Name=dyn_string_l_03, swing1LimitDegrees=45, swing2LimitDegrees=60, twistLimitDegrees=5}
 {bone1Name=dyn_string_l_05, bone2Name=dyn_string_l_04, swing1LimitDegrees=45, swing2LimitDegrees=90, twistLimitDegrees=5}
 {bone1Name=dyn_string_r_01, bone2Name=spine_05, swing1LimitDegrees=45, swing2LimitDegrees=45, twistLimitDegrees=5}
 {bone1Name=dyn_string_r_03, bone2Name=dyn_string_r_01, swing1LimitDegrees=45, swing2LimitDegrees=45, twistLimitDegrees=5}
 {bone1Name=dyn_string_r_04, bone2Name=dyn_string_r_03, swing1LimitDegrees=45, swing2LimitDegrees=60, twistLimitDegrees=5}
 {bone1Name=dyn_string_r_05, bone2Name=dyn_string_r_04, swing1LimitDegrees=45, swing2LimitDegrees=90, twistLimitDegrees=5}
]
