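# Physics asset for the F_MED_NeonCat face accessory T_03 mesh, setting up a simulated ponytail on a female player character.
# Five box bodies and a round tip capsule form a chain from the head, with tight twist limits and swing limits that widen toward the tip.
# Head, neck_01 and spine_05 capsules approximate the upper body; head-to-neck rotation is free and spine_05 has no constraint.
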PhysicsAsset /Game/Characters/Player/Female/Medium/Bodies/F_MED_Neoncat/Meshes/F_MED_NeonCat_Face_Acc_T_03_Physics
bodies {
 dyn_t3_ponytail_01 [{shapeType=Box, extentX=5, extentY=5, extentZ=2, center=<0.5 0 0>}]
 dyn_t3_ponytail_02 [{shapeType=Box, extentX=5, extentY=5, extentZ=2, center=<0.5 0 0>}]
 dyn_t3_ponytail_03 [{shapeType=Box, extentX=5, extentY=5, extentZ=2, center=<0.5 0 0>}]
 dyn_t3_ponytail_04 [{shapeType=Box, extentX=5, extentY=5, extentZ=3, center=<0.5 0 0>}]
 dyn_t3_ponytail_05 [{shapeType=Box, extentX=5, extentY=5, extentZ=3, center=<0.5 0 0>}]
 dyn_t3_ponytail_06 [{shapeType=Capsule, length=0, radius=3.25, center=<3.5791 0 0>}]
 head [{shapeType=Capsule, length=0, radius=11.5016, center=<3.7579 1.3495 -0.0002>}]
 neck_01 [{shapeType=Capsule, length=0.5, radius=6, rotation=<rot 0 -90 0>}]
 spine_05 [{shapeType=Capsule, length=15, radius=15, center=<-4.6116 2.3923 0>}]
}
constraints [
 {bone1Name=dyn_t3_ponytail_01, bone2Name=head, swing1LimitDegrees=22.5, swing2LimitDegrees=22.5, twistLimitDegrees=5}
 {bone1Name=dyn_t3_ponytail_02, bone2Name=dyn_t3_ponytail_01, swing1LimitDegrees=22.5, swing2LimitDegrees=22.5, twistLimitDegrees=5}
 {bone1Name=dyn_t3_ponytail_03, bone2Name=dyn_t3_ponytail_02, swing1LimitDegrees=22.5, swing2LimitDegrees=22.5, twistLimitDegrees=5}
 {bone1Name=dyn_t3_ponytail_04, bone2Name=dyn_t3_ponytail_03, swing1LimitDegrees=45, swing2LimitDegrees=22.5, twistLimitDegrees=5}
 {bone1Name=dyn_t3_ponytail_05, bone2Name=dyn_t3_ponytail_04, swing1LimitDegrees=45, swing2LimitDegrees=22.5, twistLimitDegrees=5}
 {bone1Name=dyn_t3_ponytail_06, bone2Name=dyn_t3_ponytail_05, swing1LimitDegrees=45, swing2LimitDegrees=45, twistLimitDegrees=5}
 {
  bone1Name=head
  bone2Name=neck_01
  swing1LimitDegrees=45
  swing1Motion=Free
  swing2LimitDegrees=45
  swing2Motion=Free
  twistLimitDegrees=45
  twistMotion=Free
 }
]
