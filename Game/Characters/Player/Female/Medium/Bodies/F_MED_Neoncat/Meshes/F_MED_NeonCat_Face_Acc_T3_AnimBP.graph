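# Animation Blueprint for the female NeonCat T3 face accessory mesh, which follows its attached parent's pose and adds secondary motion.
# The copied parent pose is cached, then fed into a RigidBody node simulating in root bone space with an override physics asset.
# When rigid body is disabled or LOD exceeds 1, it instead layers a baked pose onto the ponytail chain from dyn_t3_ponytail_01.
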AnimBlueprint /Game/Characters/Player/Female/Medium/Bodies/F_MED_Neoncat/Meshes/F_MED_NeonCat_Face_Acc_T3_AnimBP
target_skeleton=&/F_MED_NeonCat_Face_Acc__Skeleton
graphs [
 AnimGraph {nodes [
   n1 = RigidBody(ComponentPose=Local To Component(LocalPose=Use cached pose 'Pre RBN'())) {
    bClampLinearTranslationLimitToRefPose=true
    bOverrideWorldGravity=true
    overridePhysicsAsset=&/F_MED_NeonCat_Face_Acc_T_03_Physics
    simulationSpace=BaseBoneSpace
    alphaBoolBlend {blendOption=Linear}
    baseBoneRef {boneName=root}
    componentLinearAccScale=<0.5 0.5 0.5>
    componentLinearVelScale=<0.5 0.5 0.5>
    overrideWorldGravity {z=-980}
   }
   n2 = Sequence Player 'NeonCat_BakePose_FaceAcc'() {sequence=&/NeonCat_BakePose_FaceAcc}
   n3 = Use cached pose 'Pre RBN'()
   n4 = "integer > integer"(A=Get CurrentLOD(), B=1) {is=title K2Node_CallFunction}
   n5 = NOT Boolean(A=Get bRigidBodyEnabled())
   Output Pose(Result=Blend Poses by bool(
     BlendPose_0=Layered blend per bone(BasePose=@n3, BlendPoses_0=@n2) {layerSetup [{branchFilters [{blendDepth=1, boneName=dyn_t3_ponytail_01}]}]}
     BlendPose_1=Component To Local(ComponentPose=@n1)
     bActiveValue=OR Boolean(A=@n5, B=@n4)
    ) {blendType=Linear}) {layerGroup=None}
   Save cached pose 'Pre RBN'(Pose=Copy Pose From Mesh() {bUseAttachedParent=true}) {cacheName=Pre RBN}
  ]}
 EventGraph {nodes [
   TryGetPawnOwner()
   Event BlueprintUpdateAnimation()
  ]}
]
variables []
